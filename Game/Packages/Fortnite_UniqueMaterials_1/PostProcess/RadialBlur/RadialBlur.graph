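# Post-process material function that produces a radial or directional screen blur, with a scene-color variant for static meshes.
# Direction comes from screen center, a camera-facing-weighted world position or a constant vector.
# A sphere mask, optionally world-centered and distance-scaled, sets strength between center and edges.
# It averages 4 offset taps, or 16 at Epic quality, using temporal dither below Epic; the custom HLSL blur is unused.
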MaterialFunction /Game/Packages/Fortnite_UniqueMaterials_1/PostProcess/RadialBlur/RadialBlur
graph {nodes [
  n1 = ScalarParameter() {defaultValue=0.0272, parameterName=BlurDither}
  n2 = Constant2Vector() {g=0.5, r=0.5}
  n3 = VectorParameter() {parameterName=WorldPosition, defaultValue=<rgba 0 0 0 1>}
  n4 = MaterialFunctionCall(World Position=@n3.RGB) {materialFunction=&TransformToClipSpace}
  n5 = LinearInterpolate(A=@n2, B=@n4.Screen aligned 0-1 UV, Alpha=ScalarParameter() {parameterName=UseWorldPositionAsMaskCenter})
  n6 = ScreenPosition()
  n7 = MaterialFunctionCall(Float2=@n6.ViewportUV) {materialFunction=&BreakOutFloat2Components}
  n8 = MaterialFunctionCall(Float2=MaterialFunctionCall() {materialFunction=&ScreenResolution}.Buffer Resolution) {materialFunction=&BreakOutFloat2Components}
  n9 = Divide(A=@n8.R, B=@n8.R)
  n10 = LinearInterpolate(A=@n6.ViewportUV, B=AppendVector(A=@n7.R, B=Multiply(A=@n9, B=@n7.G)), Alpha=ScalarParameter() {parameterName=UseScaledUVToCalcMask})
  n11 = ScalarParameter() {defaultValue=0.5, parameterName=CenterMaskSize}
  n12 = Subtract(A=CameraPositionWS(), B=@n3.RGB)
  n13 = Length(None=@n12) {is=class MaterialExpressionLength}
  n14 = Max(A=Divide(A=@n13) {constB=1024})
  n15 = Divide(A=@n13, B=ScalarParameter() {parameterName=NearScaleMaskSizeDistance})
  n16 = LinearInterpolate(A=@n11, B=Multiply(A=Divide(A=@n11, B=@n14), B=Clamp(None=@n15)), Alpha=ScalarParameter() {parameterName=UseDistanceScaleMaskSize})
  n17 = ScalarParameter() {parameterName=CenterMaskhardness}
  n18 = Multiply(
   A=LinearInterpolate(
    A=ScalarParameter() {defaultValue=1, parameterName=BlurEdges}
    B=ScalarParameter() {parameterName=BlurCenter}
    Alpha=SphereMask(A=@n5, B=@n10, Radius=@n16, Hardness=@n17)
   )
   B=ScalarParameter() {defaultValue=1, parameterName=BlurControl}
  )
  n19 = LinearInterpolate(B=@n1, Alpha=@n18)
  Clamp(None=@n19, Max=@n1)
  MaterialExpressionLinearInterpolate_20 = LinearInterpolate(B=ScalarParameter() {defaultValue=7.2983, parameterName=BlurRadialMax}, Alpha=@n18)
  n20 = VectorParameter() {parameterName=ConstantDirection, defaultValue=<rgba 0 0 0 1>}
  n21 = Constant2Vector()
  n22 = ScreenPosition()
  n23 = Subtract(A=@n22.ViewportUV, B=@n4.Screen aligned 0-1 UV)
  n24 = Normalize(VectorInput=@n12)
  n25 = Multiply(A=MaterialFunctionCall() {materialFunction=&CameraDirectionVector}.Result) {constB=-1}
  n26 = Saturate(None=Power(Base=DotProduct(A=@n24, B=@n25), Exp=ScalarParameter() {defaultValue=3, parameterName=Falloff}))
  n27 = StaticSwitchParameter(
   True=MaterialFunctionCall(R=@n20.R, G=@n20.R) {materialFunction=&AppendMany}.RG
   False=StaticSwitchParameter(True=LinearInterpolate(A=@n21, B=@n23, Alpha=@n26), False=Subtract(A=@n6.ViewportUV, B=@n2)) {parameterName=UseWorldPosition}
  ) {parameterName=UseConstantDirection}
  n28 = MaterialFunctionCall(Vector=AppendVector(A=@n27, B=Constant()), Default=Constant3Vector()) {materialFunction=&SafeNormalize}
  n29 = Multiply(
   A=MaterialFunctionCall() {materialFunction=&DitherTemporalAA}.Result
   B=Multiply(A=ComponentMask(None=@n28.Result) {g=true, r=true}, B=@n19)
  )
  n30 = MaterialFunctionCall(Float2=QualitySwitch(Default=Add(A=@n29, B=@n6.ViewportUV), Epic=@n6.ViewportUV)) {materialFunction=&BreakOutFloat2Components}
  MaterialExpressionMaterialFunctionCall_5 = MaterialFunctionCall(X=Clamp(None=@n30.R), Y=Clamp(None=@n30.G)) {materialFunction=&MakeFloat2}
  MaterialExpressionSceneTexture_4 = SceneTexture(UVs=@n6.ViewportUV) {sceneTextureId=PPI_PostProcessInput0}
  MaterialExpressionConstant_5 = Constant() {r=4}
  n31 = Custom!(
   BlurDist=@MaterialExpressionLinearInterpolate_20
   ScreenUV=@MaterialExpressionMaterialFunctionCall_5.Result
   Tex=@MaterialExpressionSceneTexture_4.Color
   numSamples=@MaterialExpressionConstant_5
  ) {
   code="const float2 ScreenMult = View.ViewSizeAndInvSize.xy * View.BufferSizeAndInvSize.zw;\r\n\r\nconst int TexIndex = 14;\r\nconst float Samples[8] = {-0.02,-0.01,0,0.01,0.02,0.03,0.05,0.08};\r\n\r\nfloat2 dir = float2(0.5,0.5) - ScreenUV;\r\nfloat4 sum = float4(0.0, 0.0, 0.0, 0.0);\r\nfor(int i = 0; i<numSamples; i++)\r\n{\r\n    float2 pos = ScreenUV + dir * Samples[i] * BlurDist;\r\n    pos = clamp(pos, float2(0.0,0.0), float2(1.0, 1.0));\r\n    sum += SceneTextureLookupCommon(Parameters, pos * ScreenMult, TexIndex, false);\r\n}\r\n\r\nreturn sum / numSamples;"
   showCode=true
   inputs [
    {
     inputName=BlurDist
     input {expression=&self:MaterialExpressionLinearInterpolate_20}
    }
    {
     inputName=ScreenUV
     input {expression=&self:MaterialExpressionMaterialFunctionCall_5}
    }
    {
     inputName=Tex
     input {
      expression=&self:MaterialExpressionSceneTexture_4
      mask=1
      maskA=1
      maskB=1
      maskG=1
      maskR=1
     }
    }
    {
     inputName=numSamples
     input {expression=&self:MaterialExpressionConstant_5}
    }
   ]
  }
  Divide(A=@n31, B=EyeAdaptation().EyeAdaptation)
  n32 = Multiply(A=@n27)
  n33 = LinearInterpolate(B=ScalarParameter() {defaultValue=-0.0583, parameterName=OffsetAmount}, Alpha=@n18)
  n34 = QualitySwitch(Default=@n33, Epic=Divide(A=@n33) {constB=4})
  n35 = Multiply(A=@MaterialExpressionMaterialFunctionCall_5.Result)
  n36 = Add(A=Multiply(A=@n32, B=Multiply(A=@n34)), B=@n35)
  n37 = Add(A=Multiply(A=@n32, B=Multiply(A=@n34) {constB=2}), B=@n35)
  n38 = Add(A=Multiply(A=@n32, B=Multiply(A=@n34) {constB=3}), B=@n35)
  n39 = Add(A=Multiply(A=@n32, B=Multiply(A=@n34) {constB=4}), B=@n35)
  n40 = Add(A=Add(A=SceneColor!(UVs=@n36).RGB, B=SceneColor(UVs=@n37).RGB), B=Add(A=SceneColor(UVs=@n38).RGB, B=SceneColor(UVs=@n39).RGB))
  n41 = Add(
   A=Add(A=SceneTexture(UVs=@n36) {sceneTextureId=PPI_PostProcessInput0}.Color, B=SceneTexture(UVs=@n37) {sceneTextureId=PPI_PostProcessInput0}.Color)
   B=Add(A=SceneTexture(UVs=@n38) {sceneTextureId=PPI_PostProcessInput0}.Color, B=SceneTexture(UVs=@n39) {sceneTextureId=PPI_PostProcessInput0}.Color)
  )
  n42 = Multiply(A=@n34) {constB=5}
  n43 = SceneTexture(UVs=Add(A=Multiply(A=@n32, B=@n42), B=@n35)) {sceneTextureId=PPI_PostProcessInput0}
  n44 = Multiply(A=@n34) {constB=6}
  n45 = SceneTexture(UVs=Add(A=Multiply(A=@n32, B=@n44), B=@n35)) {sceneTextureId=PPI_PostProcessInput0}
  n46 = Multiply(A=@n34) {constB=7}
  n47 = SceneTexture(UVs=Add(A=Multiply(A=@n32, B=@n46), B=@n35)) {sceneTextureId=PPI_PostProcessInput0}
  n48 = Multiply(A=@n34) {constB=8}
  n49 = SceneTexture(UVs=Add(A=Multiply(A=@n32, B=@n48), B=@n35)) {sceneTextureId=PPI_PostProcessInput0}
  n50 = Multiply(A=@n32, B=Multiply(A=@n34) {constB=9})
  n51 = Multiply(A=@n32, B=Multiply(A=@n34) {constB=10})
  n52 = Add(
   A=SceneTexture(UVs=Add(A=@n50, B=@n35)) {sceneTextureId=PPI_PostProcessInput0}.Color
   B=SceneTexture(UVs=Add(A=@n51, B=@n35)) {sceneTextureId=PPI_PostProcessInput0}.Color
  )
  n53 = Multiply(A=@n32, B=Multiply(A=@n34) {constB=11})
  n54 = Multiply(A=@n32, B=Multiply(A=@n34) {constB=12})
  n55 = Add(
   A=SceneTexture(UVs=Add(A=@n53, B=@n35)) {sceneTextureId=PPI_PostProcessInput0}.Color
   B=SceneTexture(UVs=Add(A=@n54, B=@n35)) {sceneTextureId=PPI_PostProcessInput0}.Color
  )
  n56 = Multiply(A=@n32, B=Multiply(A=@n34) {constB=13})
  n57 = Multiply(A=@n32, B=Multiply(A=@n34) {constB=14})
  n58 = Add(
   A=SceneTexture(UVs=Add(A=@n56, B=@n35)) {sceneTextureId=PPI_PostProcessInput0}.Color
   B=SceneTexture(UVs=Add(A=@n57, B=@n35)) {sceneTextureId=PPI_PostProcessInput0}.Color
  )
  n59 = Multiply(A=@n32, B=Multiply(A=@n34) {constB=15})
  n60 = Multiply(A=@n32, B=Multiply(A=@n34) {constB=16})
  n61 = Add(
   A=SceneTexture(UVs=Add(A=@n59, B=@n35)) {sceneTextureId=PPI_PostProcessInput0}.Color
   B=SceneTexture(UVs=Add(A=@n60, B=@n35)) {sceneTextureId=PPI_PostProcessInput0}.Color
  )
  n62 = Add(A=Add(A=Add(A=@n43.Color, B=@n45.Color), B=Add(A=@n47.Color, B=@n49.Color)), B=Add(A=Add(A=@n52, B=@n55), B=Add(A=@n58, B=@n61)))
  n63 = QualitySwitch(Default=Divide(A=@n41) {constB=4}, Epic=Divide(A=Add(A=@n41, B=@n62)) {constB=16})
  n64 = ShadingPathSwitch(Default=StaticSwitchParameter(True=@n63, False=@n63) {parameterName=UseWorldPosition}, Mobile=@n63) {desc=Migrated from FeatureLevelSwitch}
  n65 = StaticBool()
  FunctionOutput(None=StaticSwitch(
    True=Divide(A=@n40) {constB=4}
    False=ShadingPathSwitch(Default=@n64, Forward=@n63, Mobile=@n63)
    Value=FunctionInput(Preview=@n65) {bUsePreviewValueAsDefault=true, inputName=IsStaticmesh, inputType=FunctionInput_StaticBool}
   ))
  Round(None=ScalarParameter() {defaultValue=2, parameterName=Samples})
  TransformPosition(None=@n22.PixelPosition) {transformType=TRANSFORMPOSSOURCE_World}
  Transform(None=AppendVector(A=@n22.PixelPosition, B=Constant())) {transformSourceType=TRANSFORMSOURCE_View}
 ]}
