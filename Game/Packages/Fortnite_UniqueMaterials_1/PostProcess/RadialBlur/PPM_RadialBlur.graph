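# Post-process material that applies a radial blur to the scene through the RadialBlur material function.
# The UseBlendLerp static switch chooses between the full blur result and a lerp from the original scene color, weighted by the Blend parameter.
# The "Not use on medium and low" static switch adds a shading path switch that passes unblurred scene color on Forward and Mobile, keeping the effect on Default and Deferred.
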Material /Game/Packages/Fortnite_UniqueMaterials_1/PostProcess/RadialBlur/PPM_RadialBlur
graph {
 properties {materialDomain=MD_PostProcess, pixelDepthOffsetMode=PDOM_Legacy, usedShadingModels=Default Lit}
 nodes [
  n1 = SceneTexture() {sceneTextureId=PPI_PostProcessInput0}
  ComponentMask(None=@n1.Color) {b=true, g=true, r=true}
  n2 = MaterialFunctionCall!(IsStaticmesh) {materialFunction=&/RadialBlur}
  n3 = StaticSwitchParameter(True=LinearInterpolate(A=@n1.Color, B=@n2.Result, Alpha=ScalarParameter() {parameterName=Blend}), False=@n2.Result) {parameterName=UseBlendLerp}
  StaticSwitchParameter(True=ShadingPathSwitch(Default=@n3, Deferred=@n3, Forward=@n1.Color, Mobile=@n1.Color), False=@n3) {parameterName=Not use on medium and low}
 ]
}
