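# Translucent unlit material that renders a radial blur effect on a static mesh, usable by Niagara mesh particles.
# The RadialBlur function, with IsStaticmesh on, feeds MF_FX_EmissiveAdjustments, while AttachMeshToTheCamera offsets the mesh in front of the camera.
# Shading path switches disable the effect on Forward and Mobile by outputting constant zeros and collapsing vertices to the object position.
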Material /Game/Packages/Fortnite_UniqueMaterials_1/PostProcess/RadialBlur/PPM_RadialBlurStaticmesh
graph {
 properties {
  bCastRayTracedShadows=false
  bUsedWithNiagaraMeshParticles=true
  blendMode=BLEND_Translucent
  pixelDepthOffsetMode=PDOM_Legacy
  shadingModel=MSM_Unlit
  usedShadingModels=Unlit
 }
 nodes [
  n1 = Constant()
  n2 = ShadingPathSwitch(
   Default=MaterialFunctionCall!(IsStaticmesh=StaticBool() {value=true}) {materialFunction=&/RadialBlur}.Result
   Mobile=@n1
  ) {desc=Migrated from FeatureLevelSwitch}
  MaterialFunctionCall(Emissive=ShadingPathSwitch(Default=@n2, Deferred=@n2, Forward=@n1, Mobile=@n1)) {materialFunction=&MF_FX_EmissiveAdjustments}
  n3 = Constant()
  n4 = ShadingPathSwitch(Default=Constant() {r=1}, Mobile=@n3) {desc=Migrated from FeatureLevelSwitch}
  ShadingPathSwitch(Default=@n4, Deferred=@n4, Forward=@n3, Mobile=@n3)
  n5 = Subtract(A=ObjectPositionWS(), B=WorldPosition().XYZ)
  ShadingPathSwitch(
   Default=MaterialFunctionCall!(Object Basis 1, Object Basis 2, Object Basis 3, Camera Space Position Offset=Constant3Vector() {constant=<rgba 10 0 0 0>}) {materialFunction=/Engine/Functions/Engine_MaterialFunctions02/WorldPositionOffset/AttachMeshToTheCamera}.World Position Offset
   Forward=@n5
   Mobile=@n5
  )
 ]
}
