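# Radial blur material function that averages four samples stepped along a vector from screen center or a projected world/particle position, faded by camera facing.
# A center-to-edge sphere mask, scaled by BlurControl and inverted fresnel, drives step size and temporal-AA dither on the sampling UVs.
# IsStaticmesh switches sampling between SceneColor and PostProcessInput0.
# A custom HLSL loop blur with eye-adaptation divide is left unconnected to the output.
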MaterialFunction /Game/Packages/Fortnite_UniqueMaterials_1/PostProcess/RadialBlur/RadialBlur_Niagara
graph {nodes [
  n1 = Constant2Vector() {g=0.5, r=0.5}
  n2 = ScreenPosition()
  n3 = ScalarParameter() {defaultValue=0.5, parameterName=CenterMaskSize}
  n4 = ScalarParameter() {parameterName=CenterMaskhardness}
  n5 = Fresnel(ExponentIn=ScalarParameter() {defaultValue=0.1, parameterName=BlurControlFresnelExponent}) {baseReflectFraction=0}
  n6 = Multiply(
   A=LinearInterpolate(
    A=ScalarParameter() {defaultValue=1, parameterName=BlurEdges}
    B=ScalarParameter() {parameterName=BlurCenter}
    Alpha=SphereMask(A=@n1, B=@n2.ViewportUV, Radius=@n3, Hardness=@n4)
   )
   B=Multiply(A=ScalarParameter() {defaultValue=1, parameterName=BlurControl}, B=OneMinus(None=@n5))
  )
  MaterialExpressionLinearInterpolate_20 = LinearInterpolate(B=ScalarParameter() {defaultValue=7.2983, parameterName=BlurRadialMax}, Alpha=@n6)
  n7 = MaterialFunctionCall() {materialFunction=&DitherTemporalAA}
  n8 = ScreenPosition()
  n9 = LinearInterpolate(
   A=VectorParameter() {parameterName=WorldPosition, defaultValue=<rgba 0 0 0 1>}.RGB
   B=ParticlePositionWS()
   Alpha=ScalarParameter() {parameterName=UseParticlePosition}
  )
  n10 = MaterialFunctionCall(World Position=@n9) {materialFunction=&TransformToClipSpace}
  n11 = Subtract(A=CameraPositionWS(), B=@n9)
  n12 = MaterialFunctionCall() {materialFunction=&CameraDirectionVector}
  n13 = Power(
   Base=DotProduct(A=Normalize(VectorInput=@n11), B=Multiply(A=@n12.Result) {constB=-1})
   Exp=ScalarParameter() {defaultValue=3, parameterName=Falloff}
  )
  n14 = StaticSwitchParameter(
   True=LinearInterpolate(A=Constant2Vector(), B=Subtract(A=@n8.ViewportUV, B=@n10.Screen aligned 0-1 UV), Alpha=Saturate(None=@n13))
   False=Subtract(A=@n2.ViewportUV, B=@n1)
  ) {parameterName=UseWorldPosition}
  n15 = AppendVector(A=@n14, B=Constant())
  n16 = Constant3Vector()
  n17 = Multiply(
   A=ComponentMask(None=MaterialFunctionCall(Vector=@n15, Default=@n16) {materialFunction=&SafeNormalize}.Result) {g=true, r=true}
   B=LinearInterpolate(B=ScalarParameter() {defaultValue=0.0272, parameterName=BlurDither}, Alpha=@n6)
  )
  n18 = MaterialFunctionCall(Float2=Add(A=Multiply(A=@n7.Result, B=@n17), B=@n2.ViewportUV)) {materialFunction=&BreakOutFloat2Components}
  MaterialExpressionNamedRerouteDeclaration_0 = NamedRerouteDeclaration(None=MaterialFunctionCall(X=Clamp(None=@n18.R), Y=Clamp(None=@n18.G)) {materialFunction=&MakeFloat2}.Result) {name=DitheredScreenUV, nodeColor=<rgba 0.0078 0 1 1>}
  MaterialExpressionSceneTexture_4 = SceneTexture(UVs=ScreenPosition().ViewportUV) {sceneTextureId=PPI_PostProcessInput0}
  MaterialExpressionConstant_5 = Constant() {r=4}
  n19 = Custom!(
   BlurDist=@MaterialExpressionLinearInterpolate_20
   ScreenUV=@MaterialExpressionNamedRerouteDeclaration_0
   Tex=@MaterialExpressionSceneTexture_4.Color
   numSamples=@MaterialExpressionConstant_5
  ) {
   code="const float2 ScreenMult = View.ViewSizeAndInvSize.xy * View.BufferSizeAndInvSize.zw;\r\n\r\nconst int TexIndex = 14;\r\nconst float Samples[8] = {-0.02,-0.01,0,0.01,0.02,0.03,0.05,0.08};\r\n\r\nfloat2 dir = float2(0.5,0.5) - ScreenUV;\r\nfloat4 sum = float4(0.0, 0.0, 0.0, 0.0);\r\nfor(int i = 0; i<numSamples; i++)\r\n{\r\n    float2 pos = ScreenUV + dir * Samples[i] * BlurDist;\r\n    pos = clamp(pos, float2(0.0,0.0), float2(1.0, 1.0));\r\n    sum += SceneTextureLookup(pos * ScreenMult, TexIndex, false);\r\n}\r\n\r\nreturn sum / numSamples;"
   inputs [
    {
     inputName=BlurDist
     input {expression=&self:MaterialExpressionLinearInterpolate_20}
    }
    {
     inputName=ScreenUV
     input {expression=&self:MaterialExpressionNamedRerouteDeclaration_0}
    }
    {
     inputName=Tex
     input {
      expression=&self:MaterialExpressionSceneTexture_4
      mask=1
      maskA=1
      maskB=1
      maskG=1
      maskR=1
     }
    }
    {
     inputName=numSamples
     input {expression=&self:MaterialExpressionConstant_5}
    }
   ]
  }
  Divide(A=@n19, B=EyeAdaptation().EyeAdaptation)
  n20 = LinearInterpolate(B=ScalarParameter() {defaultValue=-0.0583, parameterName=OffsetAmount}, Alpha=@n6)
  n21 = NamedRerouteUsage()
  n22 = Add(A=Multiply(A=@n14, B=Multiply(A=@n20)), B=@n21)
  n23 = Add(A=Multiply(A=@n14, B=Multiply(A=@n20) {constB=2}), B=@n21)
  n24 = Add(A=Multiply(A=@n14, B=Multiply(A=@n20) {constB=3}), B=@n21)
  n25 = Add(A=@n21, B=Multiply(A=@n14, B=Multiply(A=@n20) {constB=4}))
  n26 = Add(A=Add(A=SceneColor!(UVs=@n22).RGB, B=SceneColor(UVs=@n23).RGB), B=Add(A=SceneColor(UVs=@n24).RGB, B=SceneColor(UVs=@n25).RGB))
  n27 = Add(
   A=Add(A=SceneTexture(UVs=@n22) {sceneTextureId=PPI_PostProcessInput0}.Color, B=SceneTexture(UVs=@n23) {sceneTextureId=PPI_PostProcessInput0}.Color)
   B=Add(A=SceneTexture(UVs=@n24) {sceneTextureId=PPI_PostProcessInput0}.Color, B=SceneTexture(UVs=@n25) {sceneTextureId=PPI_PostProcessInput0}.Color)
  )
  n28 = StaticBool()
  FunctionOutput(None=StaticSwitch(
    True=Divide(A=@n26) {constB=4}
    False=Divide(A=@n27) {constB=4}
    Value=FunctionInput(Preview=@n28) {bUsePreviewValueAsDefault=true, inputName=IsStaticmesh, inputType=FunctionInput_StaticBool}
   ))
  TransformPosition(None=@n8.PixelPosition) {transformType=TRANSFORMPOSSOURCE_World}
  Transform(None=AppendVector(A=@n8.PixelPosition, B=Constant())) {transformSourceType=TRANSFORMSOURCE_View}
 ]}
