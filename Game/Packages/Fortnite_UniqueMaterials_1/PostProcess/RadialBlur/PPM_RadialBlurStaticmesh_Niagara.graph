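# Additive unlit radial blur material for Niagara mesh particles, built on the RadialBlur_Niagara function and scaled by particle alpha.
# With UseAsCameraFX? on (default), the mesh attaches to the camera.
# When off, particle color tints it and inverted Fresnel, an optional screen-aligned texture mask and depth fade shape it in world.
# Forward and mobile paths zero the effect, and in camera mode collapse the mesh to its particle pivot.
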Material /Game/Packages/Fortnite_UniqueMaterials_1/PostProcess/RadialBlur/PPM_RadialBlurStaticmesh_Niagara
graph {
 properties {
  bCastRayTracedShadows=false
  bUsedWithNiagaraMeshParticles=true
  blendMode=BLEND_Additive
  pixelDepthOffsetMode=PDOM_Legacy
  shadingModel=MSM_Unlit
  usedShadingModels=Unlit
 }
 nodes [
  n1 = StaticBool() {value=true}
  MaterialFunctionCall!(IsStaticmesh=@n1) {materialFunction=&/RadialBlur}
  n2 = Constant()
  n3 = ShadingPathSwitch(
   Default=MaterialFunctionCall!(IsStaticmesh=@n1) {materialFunction=&/RadialBlur_Niagara}.Result
   Mobile=@n2
  ) {desc=Migrated from FeatureLevelSwitch}
  n4 = ShadingPathSwitch(Default=@n3, Deferred=@n3, Forward=@n2, Mobile=@n2)
  n5 = ParticleColor()
  MaterialFunctionCall(Emissive=StaticSwitchParameter(True=@n4, False=Multiply(A=@n5.RGB, B=@n4)) {defaultValue=true, parameterName=UseAsCameraFX?}) {materialFunction=&MF_FX_EmissiveAdjustments}
  n6 = Constant3Vector() {constant=<rgba 10 0 0 0>}
  n7 = Subtract(
   A=MaterialFunctionCall() {materialFunction=&ObjectPivotPoint}.Mesh Particle Pivot Location
   B=WorldPosition().XYZ
  )
  StaticSwitchParameter(
   True=ShadingPathSwitch(
    Default=MaterialFunctionCall!(Object Basis 1, Object Basis 2, Object Basis 3, Camera Space Position Offset=@n6) {materialFunction=/Engine/Functions/Engine_MaterialFunctions02/WorldPositionOffset/AttachMeshToTheCamera}.World Position Offset
    Forward=@n7
    Mobile=@n7
   )
   False=Constant()
  ) {defaultValue=true, parameterName=UseAsCameraFX?}
  n8 = Constant()
  n9 = ShadingPathSwitch(Default=Constant() {r=1}, Mobile=@n8) {desc=Migrated from FeatureLevelSwitch}
  n10 = Multiply(A=@n5.A, B=ShadingPathSwitch(Default=@n9, Deferred=@n9, Forward=@n8, Mobile=@n8))
  n11 = OneMinus(None=Fresnel(ExponentIn=DynamicParameter() {defaultValue=<rgba 0.1 1 1 1>, paramNames [FresnelExponent, Param2, Param3, Param4]}.FresnelExponent) {baseReflectFraction=0})
  n12 = MaterialFunctionCall(
   UVs=MaterialFunctionCall!(Center Point World Position=ParticlePositionWS()) {materialFunction=/Game/Packages/Fortnite_UniqueMaterials_1/Effects/WorldAlignedScreenCoordinates}.Result
   Texture Scale=Multiply(
    A=MaterialFunctionCall() {materialFunction=&ObjectScale}.Scale X
    B=ScalarParameter() {defaultValue=150, parameterName=MaskTextureScale}
   )
  ) {materialFunction=&ScaleUVsByCenter}
  n13 = StaticSwitchParameter(
   True=Multiply(
    A=@n11
    B=TextureSampleParameter2D(UVs=@n12.UVs) {parameterName=MaskTexture, texture=/Game/Effects/Fort_Effects/Textures/Flares/T_Flare_Streaks}.G
   )
   False=@n11
  ) {parameterName=UseMaskTexture}
  n14 = ScalarParameter() {defaultValue=1, parameterName=BaseOpacity}
  n15 = ScalarParameter() {defaultValue=100, parameterName=DepthFade}
  StaticSwitchParameter(True=@n10, False=Multiply(A=Multiply(A=@n10, B=@n13), B=DepthFade(Opacity=@n14, FadeDistance=@n15))) {defaultValue=true, parameterName=UseAsCameraFX?}
 ]
}
